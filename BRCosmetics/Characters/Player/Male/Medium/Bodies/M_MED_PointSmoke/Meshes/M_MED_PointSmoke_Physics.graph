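# Physics asset for the PointSmoke medium male player body mesh, set up for a dangling belt rather than a full-body ragdoll.
# A sphere on the belt master bone roots a chain of two belt capsules, joined by constraints with matching swing limits and a zero twist limit.
# Capsules on the pelvis and right thigh have no constraints, apparently serving as collision volumes for the belt.
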PhysicsAsset /BRCosmetics/Characters/Player/Male/Medium/Bodies/M_MED_PointSmoke/Meshes/M_MED_PointSmoke_Physics
bodies {
 dyn_belt_01 [{shapeType=Capsule, length=6.5, radius=2, shapeName=dyn_belt_01_capsule, center=<4.5479 0 0>, rotation=<rot 90 0 -0.0002>}]
 dyn_belt_02 [{shapeType=Capsule, length=7.5, radius=2, shapeName=dyn_belt_02_capsule, center=<4.5479 0 0>, rotation=<rot 90 6.7213 6.7211>}]
 dyn_belt_mstr [{shapeType=Sphere, radius=0.505, shapeName=dyn_belt_mstr_sphere}]
 pelvis [{shapeType=Capsule, length=20.4851, radius=8.1311, shapeName=pelvis_capsule, center=<4.3 0 0>}]
 thigh_r [{shapeType=Capsule, length=18.4456, radius=9.2199, shapeName=thigh_r_capsule, center=<7.962 -0.0312 -0.2203>, rotation=<rot 90 180 -179.9998>}]
}
constraints [
 {bone1Name=dyn_belt_01, bone2Name=dyn_belt_mstr, swing1LimitDegrees=45, swing2LimitDegrees=45, twistLimitDegrees=0}
 {bone1Name=dyn_belt_02, bone2Name=dyn_belt_01, swing1LimitDegrees=45, swing2LimitDegrees=45, twistLimitDegrees=0}
]
