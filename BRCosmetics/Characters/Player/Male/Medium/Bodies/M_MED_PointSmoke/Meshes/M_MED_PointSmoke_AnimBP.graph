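# Animation Blueprint for the M_MED_PointSmoke player body cosmetic, adding secondary motion to jacket, strings, hood and belt bag.
# It copies the parent pose, layers HeadMesh jaw bones, then runs a crouch-aware Control Rig and RigidBody into a cached pose.
# bIsInFrontEnd selects between world-space and root-relative AnimDynamics chains, followed by hood and belt-bag dynamics.
# A DefaultSlot layer from thigh_twist_01_r is weighted by the lobby idle curve.
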AnimBlueprint /BRCosmetics/Characters/Player/Male/Medium/Bodies/M_MED_PointSmoke/Meshes/M_MED_PointSmoke_AnimBP
target_skeleton=&/M_MED_PointSmoke_Skeleton
graphs [
 AnimGraph {nodes [
   n1 = Anim Dynamics - Bone: dyn_jacket_l_01(ComponentPose=Local To Component(LocalPose=Use cached pose 'PostRBAn'()), Alpha=1, ExternalForce="0,0,0") {
    angularDampingOverride=0.999
    bOverrideAngularDamping=true
    bUseGravityOverride=true
    simulationSpace=RootRelative
    alphaBoolBlend {blendOption=Linear}
    boundBone {boneName=dyn_jacket_l_01}
    componentAppliedLinearAccClamp=<3000 3000 3000>
    componentLinearAccScale=<0.4 0.4 0.4>
    componentLinearVelScale=<0.4 0.4 0.4>
    gravityOverride {z=-980}
    physicsBodyDefinitions [{
      boundBone {boneName=dyn_jacket_l_01}
      boxExtents=<10 10 10>
      constraintSetup {angularLimitsMax=<0 15 30>, angularLimitsMin=<0 -15 0>, angularTarget=<0 0 0>}
      localJointOffset=<1 0 0>
     }]
   }
   n2 = Anim Dynamics - Bone: dyn_jacket_l_02(ComponentPose=@n1, Alpha=1, ExternalForce="0,0,0") {
    angularDampingOverride=0.999
    bOverrideAngularDamping=true
    bUseGravityOverride=true
    simulationSpace=RootRelative
    alphaBoolBlend {blendOption=Linear}
    boundBone {boneName=dyn_jacket_l_02}
    componentAppliedLinearAccClamp=<3000 3000 3000>
    componentLinearAccScale=<0.4 0.4 0.4>
    componentLinearVelScale=<0.4 0.4 0.4>
    gravityOverride {z=-980}
    physicsBodyDefinitions [{
      boundBone {boneName=dyn_jacket_l_02}
      boxExtents=<10 10 10>
      constraintSetup {angularLimitsMax=<0 15 30>, angularLimitsMin=<0 -15 0>, angularTarget=<0 0 0>}
      localJointOffset=<1 0 0>
     }]
   }
   n3 = Anim Dynamics - Bone: dyn_jacket_r_01(ComponentPose=@n2, Alpha=1, ExternalForce="0,0,0") {
    angularDampingOverride=0.999
    bOverrideAngularDamping=true
    bUseGravityOverride=true
    simulationSpace=RootRelative
    alphaBoolBlend {blendOption=Linear}
    boundBone {boneName=dyn_jacket_r_01}
    componentAppliedLinearAccClamp=<3000 3000 3000>
    componentLinearAccScale=<0.4 0.4 0.4>
    componentLinearVelScale=<0.4 0.4 0.4>
    gravityOverride {z=-980}
    physicsBodyDefinitions [{
      boundBone {boneName=dyn_jacket_r_01}
      boxExtents=<10 10 10>
      constraintSetup {angularLimitsMax=<0 15 30>, angularLimitsMin=<0 -15 0>, angularTarget=<0 0 0>}
      localJointOffset=<1 0 0>
     }]
   }
   n4 = Anim Dynamics - Bone: dyn_jacket_r_02(ComponentPose=@n3, Alpha=1, ExternalForce="0,0,0") {
    angularDampingOverride=0.999
    bOverrideAngularDamping=true
    bUseGravityOverride=true
    simulationSpace=RootRelative
    alphaBoolBlend {blendOption=Linear}
    boundBone {boneName=dyn_jacket_r_02}
    componentAppliedLinearAccClamp=<3000 3000 3000>
    componentLinearAccScale=<0.4 0.4 0.4>
    componentLinearVelScale=<0.4 0.4 0.4>
    gravityOverride {z=-980}
    physicsBodyDefinitions [{
      boundBone {boneName=dyn_jacket_r_02}
      boxExtents=<10 10 10>
      constraintSetup {angularLimitsMax=<0 15 30>, angularLimitsMin=<0 -15 0>, angularTarget=<0 0 0>}
      localJointOffset=<1 0 0>
     }]
   }
   n5 = Anim Dynamics - Bone: dyn_string_01(ComponentPose=@n4, Alpha=1, ExternalForce="0,0,0") {
    angularDampingOverride=0.999
    bOverrideAngularDamping=true
    bUseGravityOverride=true
    simulationSpace=RootRelative
    alphaBoolBlend {blendOption=Linear}
    boundBone {boneName=dyn_string_01}
    componentAppliedLinearAccClamp=<3000 3000 3000>
    componentLinearAccScale=<0.4 0.4 0.4>
    componentLinearVelScale=<0.4 0.4 0.4>
    gravityOverride {z=-400}
    physicsBodyDefinitions [{
      boundBone {boneName=dyn_string_01}
      boxExtents=<10 10 10>
      constraintSetup {angularLimitsMax=<0 20 0>, angularLimitsMin=<0 -20 -10>, angularTarget=<0 0 0>}
      localJointOffset=<1 0 0>
     }]
   }
   n6 = Anim Dynamics - Bone: dyn_string_02_r(ComponentPose=@n5, Alpha=1, ExternalForce="0,0,0") {
    angularDampingOverride=0.999
    bOverrideAngularDamping=true
    bUseGravityOverride=true
    simulationSpace=RootRelative
    alphaBoolBlend {blendOption=Linear}
    boundBone {boneName=dyn_string_02_r}
    componentAppliedLinearAccClamp=<3000 3000 3000>
    componentLinearAccScale=<0.4 0.4 0.4>
    componentLinearVelScale=<0.4 0.4 0.4>
    gravityOverride {z=-400}
    physicsBodyDefinitions [{
      boundBone {boneName=dyn_string_02_r}
      boxExtents=<10 10 10>
      constraintSetup {angularLimitsMax=<0 40 0>, angularLimitsMin=<0 -40 -5>, angularTarget=<0 0 0>}
      localJointOffset=<3 0 0>
     }]
   }
   n7 = Anim Dynamics - Bone: dyn_string_02_l(ComponentPose=@n6, Alpha=1, ExternalForce="0,0,0") {
    angularDampingOverride=0.999
    bOverrideAngularDamping=true
    bUseGravityOverride=true
    simulationSpace=RootRelative
    alphaBoolBlend {blendOption=Linear}
    boundBone {boneName=dyn_string_02_l}
    componentAppliedLinearAccClamp=<3000 3000 3000>
    componentLinearAccScale=<0.4 0.4 0.4>
    componentLinearVelScale=<0.4 0.4 0.4>
    gravityOverride {z=-200}
    physicsBodyDefinitions [{
      boundBone {boneName=dyn_string_02_l}
      boxExtents=<10 10 10>
      constraintSetup {angularLimitsMax=<0 40 0>, angularLimitsMin=<0 -40 -5>, angularTarget=<0 0 0>}
      localJointOffset=<2.5 0 0>
     }]
   }
   n8 = Anim Dynamics - Bone: dyn_jacket_back_01!(ComponentPose=@n7, bAlphaBoolEnabled:bool=True, ExternalForce="0,0,0") -> (Pose) {
    alphaInputType=Bool
    angularDampingOverride=0.999
    bOverrideAngularDamping=true
    bUseGravityOverride=true
    simulationSpace=RootRelative
    alphaBoolBlend {blendInTime=0.1, blendOption=Linear, blendOutTime=0.1}
    boundBone {boneName=dyn_jacket_back_01}
    componentAppliedLinearAccClamp=<3000 3000 3000>
    componentLinearAccScale=<0.4 0.4 0.4>
    componentLinearVelScale=<0.4 0.4 0.4>
    gravityOverride {z=-980}
    physicsBodyDefinitions [{
      boundBone {boneName=dyn_jacket_back_01}
      boxExtents=<10 10 10>
      constraintSetup {angularLimitsMax=<0 15 60>, angularLimitsMin=<0 -15 0>, angularTarget=<0 0 0>}
      localJointOffset=<1 0 0>
     }]
   }
   n9 = Anim Dynamics - Bone: dyn_jacket_back_02!(ComponentPose=@n8, bAlphaBoolEnabled:bool=True, ExternalForce="0,0,0") -> (Pose) {
    alphaInputType=Bool
    angularDampingOverride=0.999
    bOverrideAngularDamping=true
    bUseGravityOverride=true
    simulationSpace=RootRelative
    alphaBoolBlend {blendInTime=0.1, blendOption=Linear, blendOutTime=0.1}
    boundBone {boneName=dyn_jacket_back_02}
    componentAppliedLinearAccClamp=<3000 3000 3000>
    componentLinearAccScale=<0.4 0.4 0.4>
    componentLinearVelScale=<0.4 0.4 0.4>
    gravityOverride {z=-980}
    physicsBodyDefinitions [{
      boundBone {boneName=dyn_jacket_back_02}
      boxExtents=<10 10 10>
      constraintSetup {angularLimitsMax=<0 15 60>, angularLimitsMin=<0 -15 0>, angularTarget=<0 0 0>}
      localJointOffset=<1 0 0>
     }]
   }
   n10 = Component To Local(ComponentPose=@n9)
   n11 = Anim Dynamics - Bone: dyn_jacket_l_01!(ComponentPose=Local To Component(LocalPose=Use cached pose 'PostRBAn'()), ExternalForce="0,0,0") -> (Pose) {
    alpha=0.5
    angularDampingOverride=0.9999
    bOverrideAngularDamping=true
    bUseGravityOverride=true
    simulationSpace=World
    alphaBoolBlend {blendOption=Linear}
    boundBone {boneName=dyn_jacket_l_01}
    componentAppliedLinearAccClamp=<1000 1000 1000>
    gravityOverride {z=-980}
    physicsBodyDefinitions [{
      boundBone {boneName=dyn_jacket_l_01}
      boxExtents=<10 10 10>
      constraintSetup {angularLimitsMax=<0 15 30>, angularLimitsMin=<0 -15 0>, angularTarget=<0 0 0>}
      localJointOffset=<1 0 0>
     }]
   }
   n12 = Anim Dynamics - Bone: dyn_jacket_l_02!(ComponentPose=@n11, ExternalForce="0,0,0") -> (Pose) {
    alpha=0.5
    angularDampingOverride=0.9999
    bOverrideAngularDamping=true
    bUseGravityOverride=true
    simulationSpace=World
    alphaBoolBlend {blendOption=Linear}
    boundBone {boneName=dyn_jacket_l_02}
    componentAppliedLinearAccClamp=<1000 1000 1000>
    gravityOverride {z=-980}
    physicsBodyDefinitions [{
      boundBone {boneName=dyn_jacket_l_02}
      boxExtents=<10 10 10>
      constraintSetup {angularLimitsMax=<0 15 30>, angularLimitsMin=<0 -15 0>, angularTarget=<0 0 0>}
      localJointOffset=<1 0 0>
     }]
   }
   n13 = Anim Dynamics - Bone: dyn_jacket_r_01!(ComponentPose=@n12, ExternalForce="0,0,0") -> (Pose) {
    alpha=0.5
    angularDampingOverride=0.9999
    bOverrideAngularDamping=true
    bUseGravityOverride=true
    simulationSpace=World
    alphaBoolBlend {blendOption=Linear}
    boundBone {boneName=dyn_jacket_r_01}
    componentAppliedLinearAccClamp=<1000 1000 1000>
    gravityOverride {z=-980}
    physicsBodyDefinitions [{
      boundBone {boneName=dyn_jacket_r_01}
      boxExtents=<10 10 10>
      constraintSetup {angularLimitsMax=<0 15 30>, angularLimitsMin=<0 -15 0>, angularTarget=<0 0 0>}
      localJointOffset=<1 0 0>
     }]
   }
   n14 = Anim Dynamics - Bone: dyn_jacket_r_02!(ComponentPose=@n13, ExternalForce="0,0,0") -> (Pose) {
    alpha=0.5
    angularDampingOverride=0.9999
    bOverrideAngularDamping=true
    bUseGravityOverride=true
    simulationSpace=World
    alphaBoolBlend {blendOption=Linear}
    boundBone {boneName=dyn_jacket_r_02}
    componentAppliedLinearAccClamp=<1000 1000 1000>
    gravityOverride {z=-980}
    physicsBodyDefinitions [{
      boundBone {boneName=dyn_jacket_r_02}
      boxExtents=<10 10 10>
      constraintSetup {angularLimitsMax=<0 15 30>, angularLimitsMin=<0 -15 0>, angularTarget=<0 0 0>}
      localJointOffset=<1 0 0>
     }]
   }
   n15 = Anim Dynamics - Bone: dyn_string_01!(ComponentPose=@n14, ExternalForce="0,0,0") -> (Pose) {
    alpha=0.5
    angularDampingOverride=0.9999
    bOverrideAngularDamping=true
    bUseGravityOverride=true
    simulationSpace=World
    alphaBoolBlend {blendOption=Linear}
    boundBone {boneName=dyn_string_01}
    componentAppliedLinearAccClamp=<1000 1000 1000>
    gravityOverride {z=-400}
    physicsBodyDefinitions [{
      boundBone {boneName=dyn_string_01}
      boxExtents=<10 10 10>
      constraintSetup {angularLimitsMax=<0 20 0>, angularLimitsMin=<0 -20 -10>, angularTarget=<0 0 0>}
      localJointOffset=<1 0 0>
     }]
   }
   n16 = Anim Dynamics - Bone: dyn_string_02_r!(ComponentPose=@n15, ExternalForce="0,0,0") -> (Pose) {
    alpha=0.5
    angularDampingOverride=0.9999
    bOverrideAngularDamping=true
    bUseGravityOverride=true
    simulationSpace=World
    alphaBoolBlend {blendOption=Linear}
    boundBone {boneName=dyn_string_02_r}
    componentAppliedLinearAccClamp=<1000 1000 1000>
    gravityOverride {z=-400}
    physicsBodyDefinitions [{
      boundBone {boneName=dyn_string_02_r}
      boxExtents=<10 10 10>
      constraintSetup {angularLimitsMax=<0 40 0>, angularLimitsMin=<0 -40 -5>, angularTarget=<0 0 0>}
      localJointOffset=<3 0 0>
     }]
   }
   n17 = Anim Dynamics - Bone: dyn_string_02_l!(ComponentPose=@n16, ExternalForce="0,0,0") -> (Pose) {
    alpha=0.5
    angularDampingOverride=0.9999
    bOverrideAngularDamping=true
    bUseGravityOverride=true
    simulationSpace=World
    alphaBoolBlend {blendOption=Linear}
    boundBone {boneName=dyn_string_02_l}
    componentAppliedLinearAccClamp=<1000 1000 1000>
    gravityOverride {z=-200}
    physicsBodyDefinitions [{
      boundBone {boneName=dyn_string_02_l}
      boxExtents=<10 10 10>
      constraintSetup {angularLimitsMax=<0 40 0>, angularLimitsMin=<0 -40 -5>, angularTarget=<0 0 0>}
      localJointOffset=<2.5 0 0>
     }]
   }
   n18 = Anim Dynamics - Bone: dyn_jacket_back_01!(ComponentPose=@n17, ExternalForce="0,0,0") -> (Pose) {
    alpha=0.5
    angularDampingOverride=1
    bOverrideAngularDamping=true
    bUseGravityOverride=true
    simulationSpace=World
    alphaBoolBlend {blendInTime=0.1, blendOption=Linear, blendOutTime=0.1}
    boundBone {boneName=dyn_jacket_back_01}
    componentAppliedLinearAccClamp=<1000 1000 1000>
    gravityOverride {z=-1500}
    physicsBodyDefinitions [{
      boundBone {boneName=dyn_jacket_back_01}
      boxExtents=<10 10 10>
      constraintSetup {angularLimitsMax=<0 15 0>, angularLimitsMin=<0 -15 0>, angularTarget=<0 0 0>}
      localJointOffset=<1 0 0>
     }]
   }
   n19 = Anim Dynamics - Bone: dyn_jacket_back_02!(ComponentPose=@n18, ExternalForce="0,0,0") -> (Pose) {
    alpha=0.5
    angularDampingOverride=1
    bOverrideAngularDamping=true
    bUseGravityOverride=true
    simulationSpace=World
    alphaBoolBlend {blendInTime=0.1, blendOption=Linear, blendOutTime=0.1}
    boundBone {boneName=dyn_jacket_back_02}
    componentAppliedLinearAccClamp=<1000 1000 1000>
    gravityOverride {z=-1500}
    physicsBodyDefinitions [{
      boundBone {boneName=dyn_jacket_back_02}
      boxExtents=<10 10 10>
      constraintSetup {angularLimitsMax=<0 15 0>, angularLimitsMin=<0 -15 0>, angularTarget=<0 0 0>}
      localJointOffset=<1 0 0>
     }]
   }
   n20 = Component To Local(ComponentPose=@n19)
   n21 = Get bIsInFrontEnd()
   n22 = Anim Dynamics - Bone: dyn_hood_02(
    ComponentPose=Local To Component(LocalPose=Blend Poses by bool(BlendPose_0=@n20, BlendPose_1=@n10, bActiveValue=@n21) {blendType=Linear})
    Alpha=1
    ExternalForce="0,0,0"
   ) {
    angularDampingOverride=0.95
    angularSpringConstant=1000
    bAngularSpring=true
    bOverrideAngularDamping=true
    bUseGravityOverride=true
    bUsePlanarLimit=false
    simulationSpace=RootRelative
    alphaBoolBlend {blendOption=Linear}
    boundBone {boneName=dyn_hood_02}
    componentAppliedLinearAccClamp=<3000 3000 3000>
    componentLinearAccScale=<0.5 0.5 0.5>
    componentLinearVelScale=<0.5 0.5 0.5>
    gravityOverride {z=-300}
    physicsBodyDefinitions [{
      boundBone {boneName=dyn_hood_02}
      boxExtents=<10 10 15>
      constraintSetup {angularLimitsMax=<0 45 45>, angularLimitsMin=<0 -45 -5>, angularTarget=<0 0 0>, linearAxesMax=<0 0 1>, linearAxesMin=<-1 0 -1>}
      localJointOffset=<1.6 0 0>
     }]
   }
   n23 = "Transform (Modify) Bone - Bone: dyn_beltbag_01"!(ComponentPose=@n22, bAlphaBoolEnabled:bool=True, Translation="0,0,-0.5", Rotation="10,0,0", Scale="1,1,1") -> (Pose) {
    alphaInputType=Bool
    rotationMode=BMM_Additive
    rotationSpace=BCS_ParentBoneSpace
    translationMode=BMM_Additive
    translationSpace=BCS_ParentBoneSpace
    alphaBoolBlend {blendInTime=0.2, blendOption=Linear, blendOutTime=0.8}
    boneToModify {boneName=dyn_beltbag_01}
    rotation {pitch=10}
    translation {z=-0.5}
   }
   n24 = Anim Dynamics - Bone: dyn_beltbag_01!(ComponentPose=@n23, Alpha=1, GravityOverride:Vector="0,0,-900", ExternalForce="0,0,0") -> (Pose) {
    bUseGravityOverride=true
    bUsePlanarLimit=false
    simulationSpace=RootRelative
    alphaBoolBlend {blendOption=Linear}
    boundBone {boneName=dyn_beltbag_01}
    componentAppliedLinearAccClamp=<1000 1000 1000>
    componentLinearAccScale=<0.2 0.2 0.2>
    componentLinearVelScale=<0.2 0.2 0.2>
    gravityOverride {z=-900}
    physicsBodyDefinitions [{
      boundBone {boneName=dyn_beltbag_01}
      boxExtents=<10 15 15>
      constraintSetup {angularLimitsMax=<0 0 0>, angularLimitsMin=<0 -60 0>, angularTarget=<0 0 0>}
      localJointOffset=<-1.2 0 0>
     }]
   }
   Output Pose(Result=Layered blend per bone(
     BasePose=Component To Local(ComponentPose=@n24)
     BlendPoses_0=Slot 'DefaultSlot'()
     BlendWeights_0=Get FE Synced Lobby Idle Float() -> (FE Synced Lobby Idle Float:float)
    ) {layerSetup [{branchFilters [{boneName=thigh_twist_01_r}]}]}) {layerGroup=None}
   n25 = Layered blend per bone(
    BasePose=Copy Pose From Mesh() {bUseAttachedParent=true}
    BlendPoses_0=Copy Pose From Mesh(SourceMeshComponent=Get HeadMesh() -> (HeadMesh:Skeletal Mesh Component Budgeted Object Reference)) {bUseAttachedParent=true}
   ) {layerSetup [{branchFilters [
       {boneName=FACIAL_C_Jaw}
       {boneName=FACIAL_C_Jawline}
       {boneName=FACIAL_C_UnderChin}
       {boneName=FACIAL_L_Jawline}
       {boneName=FACIAL_L_UnderChin}
       {boneName=FACIAL_R_Jawline}
       {boneName=FACIAL_R_UnderChin}
      ]}]}
   n26 = Control Rig!(Source=@n25, Alpha, bIsCrouching:bool=Get bIsCrouching()) -> (Pose) {
    alphaBoolBlend {blendOption=Linear}
    controlRigAssetReference {blueprintRigClass=&/M_MED_PointSmoke_CtrlRig.M_MED_PointSmoke_CtrlRig_C}
   }
   n27 = RigidBody(ComponentPose=Local To Component(LocalPose=@n26)) {
    bOverrideWorldGravity=true
    overridePhysicsAsset=&/M_MED_PointSmoke_Physics
    simulationSpace=BaseBoneSpace
    alphaBoolBlend {blendOption=Linear}
    baseBoneRef {boneName=root}
    componentAppliedLinearAccClamp=<5000 5000 4999>
    componentLinearAccScale=<0.3 0.3 0.3>
    componentLinearVelScale=<0.3 0.3 0.3>
    overrideWorldGravity {z=-980}
   }
   Save cached pose 'PostRBAn'(Pose=Component To Local(ComponentPose=@n27)) {cacheName=PostRBAn}
  ]}
 EventGraph {nodes [
   Event BlueprintUpdateAnimation()
   > n1 = Sequence()
   > n2 = Gravity Override State Function DEPRECATED(CharacterPartAnimInstance=Self-Reference(), GameTimeInSeconds=GetGameTimeInSeconds(), GravityOverrideParams=Get Gravity Override Params())
   > Set Gravity Override(Gravity Override=@n2)
   Branch(execute=@n1.then_1, Condition=Get bIsInFrontEnd())
   > Set FE Synced Lobby Idle Float(FE Synced Lobby Idle Float:float=GetCurveValue(self=GetOwnerFrontEndAnimInstance(), CurveName=Frontend_Idle_Bypass)) -> (Output_Get:float)
   n3 = Event OnCharacterCustomizationLoaded()
   > Set HeadMesh(HeadMesh:Skeletal Mesh Component Budgeted Object Reference=Get Array Item with EFort Custom Part Type(Array=@n3.CustomMeshParts, PartType=Head)) -> (Output_Get:Skeletal Mesh Component Budgeted Object Reference)
   n4 = CheckForBackbling()
   > Set bIsNotWearingBackpack(bIsNotWearingBackpack:bool=NOT Boolean(A=OR Boolean(A=@n4, B=@n4.bIsBackblingLong))) -> (Output_Get:bool)
  ]}
]
variables [
 bIsNotWearingBackpack {type=boolean}
 FE Synced Lobby Idle Float {type=number, default=1}
 Gravity Override {type=Vector}
 Gravity Override Params {type=GravityOverrideParamsStruct}
 HeadMesh {type=SkeletalMeshComponentBudgeted}
]
