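# Physics asset for the face accessory mesh of the PointSmoke medium male player body, simulating a front-left hair strand.
# A sphere body on the head anchors a two-link chain of small capsules on the dyn_hair_front_l_a bones.
# Constraints give each link limited swing with zero twist limit, with a wider swing on the outer link than on the root.
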PhysicsAsset /BRCosmetics/Characters/Player/Male/Medium/Bodies/M_MED_PointSmoke/Meshes/Parts/M_MED_PointSmoke_FaceAcc_Physics
bodies {
 dyn_hair_front_l_a [{shapeType=Capsule, length=2, radius=1.5, center=<2.268 0 0>, rotation=<rot 90 -90 -90.0012>}]
 dyn_hair_front_l_a_01 [{shapeType=Capsule, length=2, radius=1.5, center=<2.268 0 0>, rotation=<rot 90 -90 -90.0006>}]
 head [{shapeType=Sphere, radius=8.1517, shapeName=head_sphere, center=<4.1386 0.952 0.1422>}]
}
constraints [
 {bone1Name=dyn_hair_front_l_a, bone2Name=head, swing1LimitDegrees=30, swing2LimitDegrees=30, twistLimitDegrees=0}
 {bone1Name=dyn_hair_front_l_a_01, bone2Name=dyn_hair_front_l_a, swing1LimitDegrees=45, swing2LimitDegrees=45, twistLimitDegrees=0}
]
